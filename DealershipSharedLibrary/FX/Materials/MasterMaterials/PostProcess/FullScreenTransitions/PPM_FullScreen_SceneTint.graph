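# Post-process material for full-screen transitions that recolors the scene before bloom, with a stencil-based vehicle mask.
# It desaturates and tints scene color, adds Beer's-law depth fog, world-normal fresnel and a transition edge color from MF_DLR_FullScreenTransition_Alpha.
# Custom stencil and a custom-versus-scene depth test let the masked vehicle escape the effect where unoccluded.
# A static switch selects MPC_FullScreenTransition values for Sequencer or local parameters; ES2/ES3.1 output a constant.
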Material /DealershipSharedLibrary/FX/Materials/MasterMaterials/PostProcess/FullScreenTransitions/PPM_FullScreen_SceneTint
graph {
 properties {
  bAutomaticallySetUsageInEditor=false
  bUsedWithNiagaraMeshParticles=true
  blendMode=BLEND_Translucent
  blendableLocation=BL_SceneColorBeforeBloom
  materialDomain=MD_PostProcess
  shadingModel=MSM_Unlit
  usedShadingModels=Unlit
 }
 nodes [
  n1 = SceneTexture() {sceneTextureId=PPI_PostProcessInput0}
  n2 = NamedRerouteUsage()
  n3 = NamedRerouteUsage()
  n4 = Constant()
  FeatureLevelSwitch(
   Default=LinearInterpolate(A=ComponentMask(None=@n1.Color) {b=true, g=true, r=true}, B=NamedRerouteUsage(), Alpha=Multiply(A=@n2, B=@n3))
   ES2=@n4
   ES3_1=@n4
   SM4_REMOVED=@n4
  )
  n5 = MaterialFunctionCall!(Use MPC=StaticBool() {value=true}, Is Sprite, Use World Position Behind Translucency) {materialFunction=/DealershipSharedLibrary/FX/Materials/MaterialFunctions/MF_DLR_FullScreenTransition_Alpha}
  NamedRerouteDeclaration(None=@n5.Curve - Alpha) {name=Opacity Mask, nodeColor=<rgba 0.7109 0 1 1>}
  NamedRerouteDeclaration(None=@n5.Edge Color) {name=Edge Color, nodeColor=<rgba 1 0 0.6797 1>}
  n6 = SceneTexture() {sceneTextureId=PPI_CustomStencil}
  n7 = CollectionParameter() {
   collection=/DealershipSharedLibrary/FX/Materials/MasterMaterials/PostProcess/MPC_FullScreenTransition
   parameterName=Stencil Mask - Stencil Value
  }
  n8 = ScalarParameter() {defaultValue=1, group=Vehicle Mask, parameterName=Stencil Mask - Stencil Value}
  n9 = Constant() {r=1}
  n10 = CollectionParameter() {collection=&MPC_FullScreenTransition, parameterName=Stencil Mask - Opacity}
  n11 = StaticSwitchParameter(True=@n10, False=ScalarParameter() {group=Vehicle Mask, parameterName=Stencil Mask - Opacity}) {group="!Master Controls", parameterName=Use Sequencer MPC, sortPriority=0}
  n12 = SceneTexture() {sceneTextureId=PPI_CustomDepth}
  n13 = SceneTexture() {sceneTextureId=PPI_SceneDepth}
  n14 = Saturate(None=Subtract(A=ComponentMask(None=@n12.Color) {r=true}, B=ComponentMask(None=@n13.Color) {r=true}))
  NamedRerouteDeclaration(None=If(
    A=ComponentMask(None=@n6.Color) {r=true}
    B=StaticSwitchParameter(True=@n7, False=@n8) {group="!Master Controls", parameterName=Use Sequencer MPC, sortPriority=0}
    "A > B"=@n9
    "A == B"=Max(A=@n11, B=@n14)
    "A < B"=@n9
   )) {name=Stencil Mask, nodeColor=<rgba 1 0 0.0703 1>}
  NamedRerouteDeclaration(None=ComponentMask(None=@n5.Curve - RGB) {r=true}) {name=Edge Mask, nodeColor=<rgba 1 0 0.6094 1>}
  n15 = CollectionParameter() {collection=&MPC_FullScreenTransition, parameterName=World Fresnel - Color}
  n16 = Multiply(
   A=StaticSwitchParameter(
    True=ComponentMask(None=@n15) {b=true, g=true, r=true}
    False=VectorParameter() {group=World Fresnel, parameterName=World Fresnel - Color, defaultValue=<rgba 0.1979 0.0241 0.0071 1>}.RGB
   ) {group="!Master Controls", parameterName=Use Sequencer MPC, sortPriority=0}
   B=NamedRerouteUsage()
  )
  n17 = CollectionParameter() {collection=&MPC_FullScreenTransition, parameterName=Fog - Color}
  n18 = ComponentMask(None=SceneTexture() {sceneTextureId=PPI_PostProcessInput0}.Color) {b=true, g=true, r=true}
  n19 = CollectionParameter() {collection=&MPC_FullScreenTransition, parameterName=Scene - Desaturation}
  n20 = StaticSwitchParameter(True=@n19, False=ScalarParameter() {defaultValue=1, group=Scene, parameterName=Scene - Desaturation, sortPriority=20}) {group="!Master Controls", parameterName=Use Sequencer MPC, sortPriority=0}
  n21 = ComponentMask(None=CollectionParameter() {collection=&MPC_FullScreenTransition, parameterName=Scene - Color}) {b=true, g=true, r=true}
  n22 = VectorParameter() {group=Scene, parameterName=Scene - Color Tint, defaultValue=<rgba 1 0.3217 0.1388 1>}
  n23 = LinearInterpolate(
   A=StaticSwitchParameter(
    True=ComponentMask(None=@n17) {b=true, g=true, r=true}
    False=VectorParameter() {group=Fog, parameterName=Fog - Color, defaultValue=<rgba 0.5052 0.1096 0.0515 1>}.RGB
   ) {group="!Master Controls", parameterName=Use Sequencer MPC, sortPriority=0}
   B=Multiply(
    A=Desaturation(None=@n18, Fraction=@n20)
    B=StaticSwitchParameter(True=@n21, False=@n22.RGB) {group="!Master Controls", parameterName=Use Sequencer MPC, sortPriority=0}
   )
   Alpha=NamedRerouteUsage()
  )
  n24 = ComponentMask(None=NamedRerouteUsage()) {b=true, g=true, r=true}
  n25 = NamedRerouteUsage()
  NamedRerouteDeclaration(None=Add(A=Add(A=@n23, B=@n16), B=Multiply(A=@n24, B=@n25))) {name=Final Scene, nodeColor=<rgba 0.5469 0 1 1>}
  n26 = CollectionParameter() {collection=&MPC_FullScreenTransition, parameterName=Fog - Density}
  n27 = ScalarParameter() {defaultValue=0.01, group=Fog, parameterName=Fog - Density}
  n28 = SceneTexture() {sceneTextureId=PPI_SceneDepth}
  n29 = CollectionParameter() {collection=&MPC_FullScreenTransition, parameterName=Fog - Start Distance}
  n30 = ScalarParameter() {group=Fog, parameterName=Fog - Start Distance}
  n31 = MaterialFunctionCall!(
   Thickness=Multiply(A=StaticSwitchParameter(True=@n26, False=@n27) {group="!Master Controls", parameterName=Use Sequencer MPC, sortPriority=0}) {constB=0.001}
   Depth Scale=Subtract(
    A=ComponentMask(None=@n28.Color) {r=true}
    B=StaticSwitchParameter(True=@n29, False=@n30) {group="!Master Controls", parameterName=Use Sequencer MPC, sortPriority=0}
   )
  ) {materialFunction=/Engine/Functions/Engine_MaterialFunctions01/Volumetrics/BeersLaw}
  n32 = StaticSwitchParameter(
   True=CollectionParameter() {collection=&MPC_FullScreenTransition, parameterName=Fog - Opacity}
   False=ScalarParameter() {defaultValue=1, desc="Since the fog calculations is based on the Scene Depth ", group=Fog, parameterName=Fog - Opacity}
  ) {group="!Master Controls", parameterName=Use Sequencer MPC, sortPriority=0}
  NamedRerouteDeclaration(None=LinearInterpolate(A=Saturate(None=@n31.Result), Alpha=OneMinus(None=@n32))) {name=Fog Mask, nodeColor=<rgba 0.3828 0 1 1>}
  n33 = CollectionParameter() {collection=&MPC_FullScreenTransition, parameterName=World Fresnel - Exp}
  n34 = ScalarParameter() {defaultValue=2, group=World Fresnel, parameterName=Wrold Fresnel - Exp}
  n35 = SceneTexture() {sceneTextureId=PPI_WorldNormal}
  NamedRerouteDeclaration(None=Fresnel(
    ExponentIn=StaticSwitchParameter(True=@n33, False=@n34) {group="!Master Controls", parameterName=Use Sequencer MPC, sortPriority=0}
    Normal=ComponentMask(None=@n35.Color) {b=true, g=true, r=true}
   ) {baseReflectFraction=0}) {name=World Fresnel Mask, nodeColor=<rgba 0.4297 0 1 1>}
 ]
}
